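# Unlit additive Niagara sprite material for the flower background effect of the Sleek Glasses Jacket player body cosmetic.
# A flower texture mapped through particle macro UVs is sampled per channel at slightly offset UVs for chromatic aberration, then fed into the shared FX emissive adjustments function.
# A sphere mask with a radius perturbed by panning Perlin noise is sharpened by a power and multiplied by particle alpha and depth fade.
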Material /BRCosmetics/Characters/Player/Male/Medium/Bodies/M_MED_Sleek_Glasses_Jacket/FX/M_FlowerBackground
graph {
 properties {bUsedWithNiagaraSprites=true, blendMode=BLEND_Additive, pixelDepthOffsetMode=PDOM_Legacy, shadingModel=MSM_Unlit, usedShadingModels=Unlit}
 nodes [
  n1 = Add(A=ParticleMacroUV(), B=AppendVector(A=Constant(), B=Constant() {r=0.5})) {constB=0.5}
  n2 = Constant() {r=0.001}
  n3 = Add(A=@n1, B=@n2)
  n4 = Subtract(A=@n1, B=@n2)
  MaterialFunctionCall(Emissive=MaterialFunctionCall(
    X=TextureSample(UVs=@n3) {texture=&/T_Flower_v1_A}.R
    Y=TextureSample(UVs=@n1) {texture=&T_Flower_v1_A}.G
    Z=TextureSample(UVs=@n4) {texture=&T_Flower_v1_A}.B
   ) {materialFunction=&MakeFloat3}.Result) {materialFunction=&MF_FX_EmissiveAdjustments}
  n5 = Constant() {r=0.3}
  n6 = Multiply(A=TextureSample(UVs=Panner() {speedX=0.1, speedY=0.1}) {samplerType=SAMPLERTYPE_LinearColor, texture=/Game/Effects/SnowSandInteraction/Textures/SSI_Perlin}.R) {constB=0.2}
  n7 = Power(Base=SphereMask(A=TextureCoordinate(), B=Constant2Vector() {g=0.5, r=0.5}, Radius=Add(A=@n5, B=@n6), Hardness=Constant() {r=0.5})) {constExponent=3}
  Multiply(A=Multiply(A=Saturate(None=@n7), B=ParticleColor().A), B=DepthFade() {fadeDistanceDefault=15})
 ]
}
