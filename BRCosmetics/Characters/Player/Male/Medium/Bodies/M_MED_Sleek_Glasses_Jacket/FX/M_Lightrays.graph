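# Additive unlit Niagara sprite material for rainbow light rays, stored with the Sleek Glasses Jacket character's FX.
# Two SphereMasks on split UV axes form three soft bands repeated along V with a falloff across U, weighted by inverted V.
# A red-green-blue lerp along V colors them, tinted by particle color and oversaturated through negative desaturation before emissive adjustment.
# Particle alpha, DepthFade and CameraDepthFade soften the mask near geometry and the camera.
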Material /BRCosmetics/Characters/Player/Male/Medium/Bodies/M_MED_Sleek_Glasses_Jacket/FX/M_Lightrays
graph {
 properties {bUsedWithNiagaraSprites=true, blendMode=BLEND_Additive, pixelDepthOffsetMode=PDOM_Legacy, shadingModel=MSM_Unlit, usedShadingModels=Unlit}
 nodes [
  n1 = ParticleColor()
  n2 = TextureCoordinate()
  n3 = Frac(None=TextureCoordinate() {uTiling=0, vTiling=3})
  n4 = Constant2Vector() {g=0.5}
  n5 = Constant() {r=0.6}
  n6 = Constant()
  n7 = TextureCoordinate() {vTiling=0}
  n8 = Constant2Vector() {r=0.5}
  n9 = Constant() {r=0.45}
  n10 = Constant()
  n11 = ComponentMask(None=TextureCoordinate()) {g=true}
  n12 = Multiply(
   A=Multiply(A=SphereMask(A=@n3, B=@n4, Radius=@n5, Hardness=@n6), B=SphereMask(A=@n7, B=@n8, Radius=@n9, Hardness=@n10))
   B=Add(A=OneMinus(None=@n11)) {constB=0.25}
  )
  n13 = Multiply(
   A=MaterialFunctionCall(
    1=Constant3Vector() {constant=<rgba 1 0.079 0 0>}
    2=Constant3Vector() {constant=<rgba 0 1 0.0915 0>}
    3=Constant3Vector() {constant=<rgba 0 0.1099 1 0>}
    A=ComponentMask(None=@n2) {g=true}
   ) {materialFunction=&Lerp_Multiple_Float3}.Lerp 3 Inputs
   B=@n12
  )
  MaterialFunctionCall(Emissive=Desaturation(None=Multiply(A=@n1.RGB, B=@n13), Fraction=Constant() {r=-2})) {materialFunction=&MF_FX_EmissiveAdjustments}
  Multiply(
   A=Multiply(A=Multiply(A=@n12, B=@n1.A), B=DepthFade() {fadeDistanceDefault=5})
   B=MaterialFunctionCall(Fade Length=Constant() {r=150}) {materialFunction=&CameraDepthFade}.Result
  )
 ]
}
