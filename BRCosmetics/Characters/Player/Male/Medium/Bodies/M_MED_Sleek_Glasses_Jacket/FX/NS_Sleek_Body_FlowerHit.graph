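# Niagara hit effect for the Sleek player body cosmetic that throws flowers and petals from a user-supplied skeletal mesh, alongside chroma rays and a flare.
# Four sprite emitters burst on spawn, with Flower and Petal adding a spawn rate that tapers with emitter age.
# Petals tumble under gravity and curl noise, while the flare shifts color as it fades.
# Every spawn is gated by a camera-distance threshold.
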NiagaraSystem /BRCosmetics/Characters/Player/Male/Medium/Bodies/M_MED_Sleek_Glasses_Jacket/FX/NS_Sleek_Body_FlowerHit
dependencies {
 usedDataInterfaces [
  &NiagaraDataInterfaceCamera
  &NiagaraDataInterfaceCurve
  &NiagaraDataInterfaceArrayColor
  &NiagaraDataInterfaceColorCurve
 ]
 usedDynamicInputs []
 usedModules [&SystemState]
}
emitters {
 ChromaRays {
  emitterUpdateScript [
   &EmitterState()
   &SpawnBurst_Instantaneous(
    Spawn Count=20
    Spawn Probability=&MakeCustomFloatFromBool(
     Bool=&SetBoolByFloatComparison(
      A=&VectorLength(Vector=&Subtract_Vector(
        A=Linked() {linkedVariable {name=Engine.Emitter.SimulationPosition, type=NiagaraPosition}}
        B=&ReturnCameraVector(Camera Property=::NewEnumerator0, Coordinate Space=::Simulation)
       ))
      B=175
     )
     True Float=1
     False Float=0
    )
    Use Spawn Probability=true
    UseLoopCountLimit=false
   )
  ]
  particleSpawnScript [
   &InitializeParticle(
    Lifetime Max=0.4
    Lifetime Min=0.15
    Lifetime Mode=::NewEnumerator1
    Sprite Rotation Angle=45
    Sprite Rotation Mode=::NewEnumerator0
    Sprite Size Max=<1.5 20>
    Sprite Size Min=<1.5 25>
    Sprite Size Mode=::NewEnumerator2
    UsePositionOffset=false
    Write Material Random=true
   )
   &FN_SetSpawnGroup(Number of Spawn Groups=9) {enabled=false}
   &CameraOffset(Camera Offset Amount=15)
   &AddVelocityFromPoint(Module.Use Falloff Distance=false, Velocity Strength=150)
   &SkeletalMeshLocation(Bone / Socket Index=1, Bone Sampling Mode=::NewEnumerator4, MaskSpawnGroup=false, Skeletal Mesh=$User.New Skeletal Mesh)
  ]
  particleUpdateScript [
   &Drag(Drag=5, Use Linear Drag=true, Use Rotational Drag=false)
   &SolveForcesAndVelocity(Clamp Velocity=false, Limit Acceleration=false)
   &ParticleState()
   &ScaleSpriteSize(Uniform Curve Sprite Scale=DataInterface() {propertyValues {Curve {keys [
        {arriveTangent=4.1785, interpMode=RCIM_Cubic, leaveTangent=4.1785, tangentMode=RCTM_User, time=0, value=0}
        {arriveTangent=1.9874, interpMode=RCIM_Cubic, leaveTangent=1.9874, tangentMode=RCTM_User, time=0.1, value=0.65}
        {arriveTangent=-0.0405, interpMode=RCIM_Cubic, leaveTangent=-0.0405, tangentMode=RCTM_User, time=0.5, value=1}
       ]}}})
   &ScaleColor(
    Scale Alpha=&FloatFromCurve(FloatCurve=DataInterface() {propertyValues {Curve {keys [
         <at 0 0>
         {arriveTangent=-0.4806, interpMode=RCIM_Cubic, leaveTangent=-0.4806, tangentMode=RCTM_User, time=0.1, value=0.65}
         {arriveTangent=-0.7826, interpMode=RCIM_Cubic, leaveTangent=-0.7826, tangentMode=RCTM_User, time=1, value=0}
        ]}}})
    Scale RGB=&Multiply_VectorByFloat(
     Vector=<1 1 1>
     Float=&FloatFromCurve(
      FloatCurve=DataInterface() {propertyValues {Curve {keys [
          {arriveTangent=9.0725, interpMode=RCIM_Cubic, leaveTangent=9.0725, tangentMode=RCTM_User, time=0, value=0}
          {arriveTangent=-2.5249, interpMode=RCIM_Cubic, leaveTangent=-2.5249, tangentMode=RCTM_User, time=0.15, value=1}
          {arriveTangent=-0.6053, interpMode=RCIM_Cubic, leaveTangent=-0.6053, tangentMode=RCTM_User, time=0.6, value=0.15}
          <at 1 0>
         ]}}}
      Scale Curve=3
     )
    )
    ScaleA=true
    ScaleRGB=true
   )
  ]
  rendererClasses [&NiagaraSpriteRendererProperties]
  renderers [{rendererClass=&NiagaraSpriteRendererProperties}]
 }
 Flare {
  emitterUpdateScript [
   &EmitterState()
   &SpawnBurst_Instantaneous(
    Spawn Probability=&MakeCustomFloatFromBool(
     Bool=&SetBoolByFloatComparison(
      A=&VectorLength(Vector=&Subtract_Vector(
        A=Linked() {linkedVariable {name=Engine.Emitter.SimulationPosition, type=NiagaraPosition}}
        B=&ReturnCameraVector(Camera Property=::NewEnumerator0, Coordinate Space=::Simulation)
       ))
      B=175
     )
     True Float=1
     False Float=0
    )
    Use Spawn Probability=true
    UseLoopCountLimit=false
   ) {module=SpawnBurst_Instantaneous001}
  ]
  particleSpawnScript [
   &InitializeParticle(
    Color=<rgba 1 1 1 0.3>
    Lifetime=0.2
    Sprite Size Mode=::NewEnumerator3
    Uniform Sprite Size=50
    UsePositionOffset=false
    Write Material Random=true
   )
   &CameraOffset(Camera Offset Amount=25)
   &SkeletalMeshLocation(Bone / Socket Index=1, Bone Sampling Mode=::NewEnumerator4, MaskSpawnGroup=false, Skeletal Mesh=$User.New Skeletal Mesh)
  ]
  particleUpdateScript [
   &ParticleState()
   &DynamicMaterialParameters(
    Index 0 Param 1=1
    Index 0 Param 2=&FloatFromCurve(FloatCurve=DataInterface() {propertyValues {Curve {keys [{tangentMode=RCTM_User, time=0, value=1}, {tangentMode=RCTM_User, time=1, value=0}]}}})
    Index 0 Param 4=1
    Module.Param0WriteEnabled=true
    Module.Param1WriteEnabled=true
    Module.Param2WriteEnabled=true
    Module.Param3WriteEnabled=true
   )
   &ScaleColor(
    Scale Alpha=&FloatFromCurve(FloatCurve=DataInterface() {propertyValues {Curve {keys [<at 0 1>,<at 1 0>]}}})
    Scale RGB=&MakeVectorFromLinearColorRGB(Color=&ColorFromCurve(ColorCurve=DataInterface() {propertyValues {
        AlphaCurve {keys [<at 0 1 RCIM_Cubic>,<at 1 0 RCIM_Cubic>]}
        BlueCurve {keys [
          <at 0 1 RCIM_Cubic>
          {arriveTangent=-1.596, interpMode=RCIM_Cubic, leaveTangent=-1.596, time=0.211, value=0.006}
          <at 0.5529 0.1175>
          <at 0.84 1>
         ]}
        GreenCurve {keys [<at 0 1 RCIM_Cubic>,<at 0.211 0 RCIM_Cubic>,<at 0.5529 1>,<at 0.84 0.1199>]}
        RedCurve {keys [<at 0 1 RCIM_Cubic>, {arriveTangent=-1.8085, interpMode=RCIM_Cubic, leaveTangent=-1.8085, time=0.211, value=1}, <at 0.5529 0>, <at 0.84 0>]}
       }}))
    ScaleA=true
    ScaleRGB=true
   )
  ]
  rendererClasses [&NiagaraSpriteRendererProperties]
  renderers [{rendererClass=&NiagaraSpriteRendererProperties}]
 }
 Flower {
  emitterUpdateScript [
   &EmitterState()
   &SpawnBurst_Instantaneous(
    Spawn Count=2
    Spawn Probability=&MakeCustomFloatFromBool(
     Bool=&SetBoolByFloatComparison(
      A=&VectorLength(Vector=&Subtract_Vector(
        A=Linked() {linkedVariable {name=Engine.Emitter.SimulationPosition, type=NiagaraPosition}}
        B=&ReturnCameraVector(Camera Property=::NewEnumerator0, Coordinate Space=::Simulation)
       ))
      B=175
     )
     True Float=1
     False Float=0
    )
    Use Spawn Probability=true
    UseLoopCountLimit=false
   ) {module=SpawnBurst_Instantaneous001}
   &SpawnRate(
    Spawn Probability=&MakeCustomFloatFromBool(
     Bool=&SetBoolByFloatComparison(
      A=&VectorLength(Vector=&Subtract_Vector(
        A=Linked() {linkedVariable {name=Engine.Emitter.SimulationPosition, type=NiagaraPosition}}
        B=&ReturnCameraVector(Camera Property=::NewEnumerator0, Coordinate Space=::Simulation)
       ))
      B=175
     )
     True Float=1
     False Float=0
    )
    SpawnRate=&Lerp_Float(A=10, B=0, Alpha=$Emitter.Age)
    Use Spawn Probability=true
   )
  ]
  particleSpawnScript [
   &InitializeParticle(
    Color Mode=::NewEnumerator0
    Lifetime Max=1
    Lifetime Min=0.65
    Lifetime Mode=::NewEnumerator1
    Sprite Size Mode=::NewEnumerator1
    Uniform Sprite Size Max=6
    Uniform Sprite Size Min=3
    UsePositionOffset=false
    Write Material Random=true
   )
   &SubUVAnimation(End Frame=2, SubUV Animation Mode:ENiagara_SubUVLookupMode=::NewEnumerator2)
   &SkeletalMeshLocation(Bone Sampling Mode=::NewEnumerator4, MaskSpawnGroup=false, Skeletal Mesh=$User.New Skeletal Mesh)
   &ShapeLocation(MaskSpawnGroup=false, Override Seed=false, Sphere Radius=25, UseOwnerScale=true)
   &AddVelocity(Coordinate Space=::World, Velocity=&RandomRangeVector(Minimum=<150 150 250>, Maximum=<-150 -150 150>))
   &CameraOffset(Camera Offset Amount=65)
   &GravityForce(Gravity=<0 0 -280>)
   &InheritVelocity(Speed Limit=true, UseSpeedThreshold=false)
   &ApplyInitialForces()
  ]
  particleUpdateScript [
   &Drag(Drag=5, Use Linear Drag=true, Use Rotational Drag=false)
   &SolveForcesAndVelocity(Clamp Velocity=false, Limit Acceleration=false)
   &ParticleState()
   &SpriteRotationRate(Rotation Rate=&FloatFromCurve(
     FloatCurve=DataInterface() {propertyValues {Curve {keys [
         {interpMode=RCIM_Cubic, tangentMode=RCTM_User, time=0, value=1}
         {arriveTangent=-1.7642, interpMode=RCIM_Cubic, leaveTangent=-1.7642, tangentMode=RCTM_User, time=1, value=0.2}
        ]}}}
     Scale Curve=50
    ))
   &ScaleSpriteSize(Uniform Curve Sprite Scale=DataInterface() {propertyValues {Curve {keys [
        {arriveTangent=2.816, interpMode=RCIM_Cubic, leaveTangent=2.816, tangentMode=RCTM_User, time=0, value=0.025}
        {arriveTangent=1.85, interpMode=RCIM_Cubic, leaveTangent=1.85, time=0.3, value=0.975}
        {arriveTangent=0.0833, interpMode=RCIM_Cubic, leaveTangent=0.0833, time=0.5, value=0.95}
        {arriveTangent=-1.0296, interpMode=RCIM_Cubic, leaveTangent=-1.0296, tangentMode=RCTM_User, time=0.6, value=1}
        {tangentMode=RCTM_User, time=1, value=0.025}
       ]}}})
   &SolveRotationalForcesAndVelocity()
  ]
  rendererClasses [&NiagaraSpriteRendererProperties]
  renderers [{rendererClass=&NiagaraSpriteRendererProperties}]
 }
 Petal {
  emitterUpdateScript [
   &EmitterState()
   &SpawnBurst_Instantaneous(
    Spawn Count=4
    Spawn Probability=&MakeCustomFloatFromBool(
     Bool=&SetBoolByFloatComparison(
      A=&VectorLength(Vector=&Subtract_Vector(
        A=Linked() {linkedVariable {name=Engine.Emitter.SimulationPosition, type=NiagaraPosition}}
        B=&ReturnCameraVector(Camera Property=::NewEnumerator0, Coordinate Space=::Simulation)
       ))
      B=175
     )
     True Float=1
     False Float=0
    )
    Use Spawn Probability=true
    UseLoopCountLimit=false
   ) {module=SpawnBurst_Instantaneous001}
   &SpawnRate(
    Spawn Probability=&MakeCustomFloatFromBool(
     Bool=&SetBoolByFloatComparison(
      A=&VectorLength(Vector=&Subtract_Vector(
        A=Linked() {linkedVariable {name=Engine.Emitter.SimulationPosition, type=NiagaraPosition}}
        B=&ReturnCameraVector(Camera Property=::NewEnumerator0, Coordinate Space=::Simulation)
       ))
      B=175
     )
     True Float=1
     False Float=0
    )
    SpawnRate=&Lerp_Float(A=10, B=0, Alpha=$Emitter.Age)
    Use Spawn Probability=true
   ) {module=SpawnRate001}
  ]
  particleSpawnScript [
   &InitializeParticle(
    Color=&SelectLinearColorFromArray(
     Array Sampling Mode=::NewEnumerator0
     Direct Array Index=&ReturnExecIndex()
     Color Selection Array=DataInterface() {propertyValues {ColorData [<rgba 0.5529 0.1176 0.0078 1>, <rgba 0.7686 0.6863 0.3922 1>, <rgba 0.1198 0.0496 0.029 1>]}}
     Random Seed=0
    )
    Lifetime Max=1
    Lifetime Min=0.5
    Lifetime Mode=::NewEnumerator1
    Sprite Rotation Angle=&RandomRangeFloat(Minimum=0, Maximum=1)
    Sprite Rotation Mode=::NewEnumerator1
    Sprite Size Mode=::NewEnumerator1
    Uniform Sprite Size Max=5
    Uniform Sprite Size Min=3
    UsePositionOffset=false
    Write Material Random=true
   )
   &SubUVAnimation(End Frame=3, Start Frame=3, SubUV Animation Mode:ENiagara_SubUVLookupMode=::NewEnumerator2)
   &SkeletalMeshLocation(Bone Sampling Mode=::NewEnumerator4, MaskSpawnGroup=false, Skeletal Mesh=$User.New Skeletal Mesh)
   &ShapeLocation(MaskSpawnGroup=false, Override Seed=false, Sphere Radius=5, UseOwnerScale=true)
   &AddVelocity(Coordinate Space=::World, Velocity=&RandomRangeVector(Minimum=<150 150 250>, Maximum=<-150 -150 150>))
   &CameraOffset(Camera Offset Amount=25)
   &GravityForce(Gravity=<0 0 -280>)
   &InheritVelocity(Speed Limit=true, UseSpeedThreshold=false)
   &ApplyInitialForces()
  ]
  particleUpdateScript [
   &CurlNoiseForce(Noise Strength=450, Pan Noise=false)
   &Drag(Drag=4, Use Linear Drag=true, Use Rotational Drag=false)
   &SolveForcesAndVelocity(Clamp Velocity=false, Limit Acceleration=false)
   &ParticleState()
   &ScaleSpriteSize(Uniform Curve Sprite Scale=DataInterface() {propertyValues {Curve {keys [
        {interpMode=RCIM_Cubic, tangentMode=RCTM_User, time=0, value=0}
        {arriveTangent=1.2143, interpMode=RCIM_Cubic, leaveTangent=1.2143, time=0.2, value=0.85}
        {arriveTangent=-1.0625, interpMode=RCIM_Cubic, leaveTangent=-1.0625, time=0.7, value=0.85}
        {tangentMode=RCTM_User, time=1, value=0}
       ]}}})
   &ScaleSpriteSizeBySpeed(Max Scale Factor=<0.75 1>, Min Scale Factor=<0.75 1>, Velocity Threshold=200)
  ]
  rendererClasses [&NiagaraSpriteRendererProperties]
  renderers [{rendererClass=&NiagaraSpriteRendererProperties}]
 }
}
summary {
 systemName=NS_Sleek_Body_FlowerHit
 emitters [
  {emitterName=Flower, rendererClasses [&NiagaraSpriteRendererProperties]}
  {emitterName=Petal, rendererClasses [&NiagaraSpriteRendererProperties]}
  {emitterName=ChromaRays, rendererClasses [&NiagaraSpriteRendererProperties]}
  {emitterName=Flare, rendererClasses [&NiagaraSpriteRendererProperties]}
 ]
 userVariables [User.New Skeletal Mesh {
   type=NiagaraDataInterfaceSkeletalMesh
   defaultValue {
    struct=NiagaraExt_VariableValue_DataInterface
    value {dataInterface=&self:NiagaraDataInterfaceSkeletalMesh_0}
   }
  }]
}
