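# Physics asset for the M_MED_Sleek_Glasses_Jacket male body, simulating the jacket's hanging cloth on dyn_Skirt bones.
# Seven six-bone chains (B, D–H, J) hang from spine_01 with limited swing and twist, except chain F's locked top three joints.
# Cross-constraints tie neighboring chains at most rows, forming a lattice that moves like a sheet.
# Capsules and boxes shape the chains, alongside bodies on pelvis, spine, legs, feet and root.
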PhysicsAsset /BRCosmetics/Characters/Player/Male/Medium/Bodies/M_MED_Sleek_Glasses_Jacket/Meshes/M_MED_Sleek_Glasses_Jacket_Physics
bodies {
 calf_l [{shapeType=Capsule, length=38.803, radius=9.909, shapeName=calf_l_capsule, center=<-27.0064 -0.0002 0>, rotation=<rot -90 0 0>}]
 calf_r [{shapeType=Capsule, length=38.803, radius=9.909, shapeName=calf_l_capsule, center=<25.1541 -0.0002 0>, rotation=<rot -90 0 0>}]
 dyn_SkirtB01 [{
   shapeType=Capsule
   length=4.7386
   radius=4.0379
   shapeName=dyn_SkirtJ01_capsule
   center=<0.0023 0.2489 -4.373>
   rotation=<rot -0.7126 179.9997 179.9999>
  }]
 dyn_SkirtB02 [{shapeType=Box, extentX=23.3552, extentY=5.9738, extentZ=14.4279, shapeName=dyn_SkirtB02_box, center=<0 -0.0328 -4.1981>}]
 dyn_SkirtB03 [{shapeType=Box, extentX=23.3552, extentY=5.9738, extentZ=14.4279, shapeName=dyn_SkirtB03_box, center=<0 -0.0328 -4.1981>}]
 dyn_SkirtB04 [{shapeType=Box, extentX=25.2303, extentY=5.3826, extentZ=19, shapeName=dyn_SkirtJ04_box, center=<0 0 -5.5058>}]
 dyn_SkirtB05 [{shapeType=Box, extentX=25.2303, extentY=5.3826, extentZ=19, shapeName=dyn_SkirtJ05_box, center=<0 0 -5.5058>}]
 dyn_SkirtB06 [{shapeType=Box, extentX=23.8456, extentY=5.2642, extentZ=23, shapeName=dyn_SkirtJ06_box, center=<0 0 -8.9173>}]
 dyn_SkirtD01 [{shapeType=Capsule, length=9, radius=3, shapeName=dyn_SkirtD01_capsule, center=<0 0 -4.9582>}]
 dyn_SkirtD02 [
  {shapeType=Capsule, length=8.5, radius=3.5, shapeName=dyn_SkirtD02_capsule, center=<0 0.1324 -4.943>, rotation=<rot 0 0 -6.5644>}
  {shapeType=Capsule, length=9.4554, radius=8.3793, shapeName=dyn_SkirtD02_capsule, center=<0 0.993 0>, rotation=<rot 90 0 0>}
 ]
 dyn_SkirtD03 [
  {shapeType=Capsule, length=12.75, radius=3, shapeName=dyn_SkirtD03_capsule, center=<0 0.5078 -5.1158>, rotation=<rot 0 0 -6.4159>}
  {shapeType=Capsule, length=9.4554, radius=8.3793, shapeName=dyn_SkirtD03_capsule, center=<0 0.993 0>, rotation=<rot 90 0 0>}
 ]
 dyn_SkirtD04 [
  {shapeType=Capsule, length=9.4554, radius=8.3793, shapeName=dyn_SkirtD04_capsule, center=<0 0.993 0>, rotation=<rot 90 0 0>}
  {shapeType=Box, extentX=17.5, extentY=4.5, extentZ=18, shapeName=dyn_SkirtD04_box, center=<0 0.8276 -8.4743>, rotation=<rot 0 0 -5.5799>}
 ]
 dyn_SkirtD05 [
  {shapeType=Capsule, length=9.4554, radius=8.3793, shapeName=dyn_SkirtD05_capsule, center=<0 0.993 0>, rotation=<rot 90 0 0>}
  {shapeType=Box, extentX=17.5, extentY=4.5, extentZ=18, shapeName=dyn_SkirtD05_box, center=<0 0.8276 -8.4743>, rotation=<rot 0 0 -5.5799>}
 ]
 dyn_SkirtD06 [{
   shapeType=Capsule
   length=19.7603
   radius=9.75
   shapeName=dyn_SkirtH06_capsule
   center=<0 1.5221 -3.9088>
   rotation=<rot 78.5145 -72.4721 -72.8021>
  }]
 dyn_SkirtE01 [{shapeType=Capsule, length=8.5, radius=3.5, shapeName=dyn_SkirtE01_capsule, center=<0 0 -4.9582>, rotation=<rot 0 0 -6.106>}]
 dyn_SkirtE02 [{shapeType=Capsule, length=8.5, radius=3.5, shapeName=dyn_SkirtE02_capsule, center=<0 0 -4.9582>, rotation=<rot 0 0 -6.106>}]
 dyn_SkirtE03 [{
   shapeType=Capsule
   length=7.518
   radius=7.482
   shapeName=dyn_SkirtE03_capsule
   center=<0 1.2327 -4.1412>
   rotation=<rot 89.0373 -0.0007 -0.0007>
  }]
 dyn_SkirtE04 [{
   shapeType=Capsule
   length=5.5478
   radius=9.4522
   shapeName=dyn_SkirtE04_capsule
   center=<0 0.8735 -3.9456>
   rotation=<rot 81.5595 -86.9859 -87.0185>
  }]
 dyn_SkirtE05 [{
   shapeType=Capsule
   length=5.7943
   radius=9.2057
   shapeName=dyn_SkirtE05_capsule
   center=<0 0.8735 -3.9456>
   rotation=<rot 81.5595 -86.9859 -87.0185>
  }]
 dyn_SkirtE06 [{
   shapeType=Capsule
   length=8.5
   radius=8.5
   shapeName=dyn_SkirtG06_capsule
   center=<0 1.4131 -2.2482>
   rotation=<rot 80.9682 -67.5629 -58.4206>
  }]
 dyn_SkirtF01 [{shapeType=Capsule, length=8.5, radius=3.5, shapeName=dyn_SkirtF01_capsule, center=<0 0 -4.9582>, rotation=<rot 0 0 -6.1057>}]
 dyn_SkirtF02 [{shapeType=Capsule, length=8.5, radius=3.5, shapeName=dyn_SkirtF02_capsule, center=<0 0 -4.9582>, rotation=<rot 0 0 -6.1057>}]
 dyn_SkirtF03 [{
   shapeType=Capsule
   length=6.518
   radius=8.482
   shapeName=dyn_SkirtF03_capsule
   center=<0 1.2327 -4.1412>
   rotation=<rot 89.0298 -0.0024 -0.0024>
  }]
 dyn_SkirtF04 [{
   shapeType=Capsule
   length=9.8686
   radius=8.7022
   shapeName=dyn_SkirtF04_capsule
   center=<0 0.8736 -3.9457>
   rotation=<rot 81.5636 -87.0809 -87.1124>
  }]
 dyn_SkirtF05 [{
   shapeType=Capsule
   length=6.5443
   radius=8.4557
   shapeName=dyn_SkirtF05_capsule
   center=<0 0.8736 -3.9457>
   rotation=<rot 81.5636 -87.0809 -87.1124>
  }]
 dyn_SkirtF06 [{
   shapeType=Capsule
   length=13.25
   radius=7.75
   shapeName=dyn_SkirtF06_capsule
   center=<0 -0.446 -4.1411>
   rotation=<rot 81.4863 -81.6892 -81.7795>
  }]
 dyn_SkirtG01 [{shapeType=Capsule, length=8.5, radius=3.5, shapeName=dyn_SkirtG01_capsule, center=<0 0 -4.9582>, rotation=<rot 0 0 -6.1057>}]
 dyn_SkirtG02 [{shapeType=Capsule, length=8.5, radius=3.5, shapeName=dyn_SkirtG02_capsule, center=<0 0 -4.9582>, rotation=<rot 0 0 -6.1057>}]
 dyn_SkirtG03 [{
   shapeType=Capsule
   length=10.0386
   radius=7.57
   shapeName=dyn_SkirtG03_capsule
   center=<0.0341 1.4223 -4.2373>
   rotation=<rot 82.8092 -82.2193 -82.2798>
  }]
 dyn_SkirtG04 [{
   shapeType=Capsule
   length=7.7978
   radius=8.2022
   shapeName=dyn_SkirtG04_capsule
   center=<0 1.0779 -4.2485>
   rotation=<rot 80.5595 -87.3075 -87.3438>
  }]
 dyn_SkirtG05 [{
   shapeType=Capsule
   length=8.0443
   radius=7.9557
   shapeName=dyn_SkirtG05_capsule
   center=<0 1.0779 -4.2485>
   rotation=<rot 80.5595 -87.3075 -87.3438>
  }]
 dyn_SkirtG06 [{
   shapeType=Capsule
   length=8.5
   radius=8.5
   shapeName=dyn_SkirtG06_capsule
   center=<0 1.4131 -2.2482>
   rotation=<rot 80.9682 -67.5629 -58.4206>
  }]
 dyn_SkirtH01 [{shapeType=Capsule, length=8.75, radius=3.25, shapeName=dyn_SkirtH01_capsule, center=<0 0 -4.9582>, rotation=<rot 0 0 -9.9728>}]
 dyn_SkirtH02 [{shapeType=Capsule, length=8.75, radius=3.25, shapeName=dyn_SkirtH02_capsule, center=<0.0001 1.1859 -4.8124>, rotation=<rot 0 0 -7.0095>}]
 dyn_SkirtH03 [{shapeType=Capsule, length=9.75, radius=3.5, shapeName=dyn_SkirtH03_capsule, center=<0 0.4203 -4.6784>, rotation=<rot 0 0 -5.1331>}]
 dyn_SkirtH04 [{shapeType=Box, extentX=22.5, extentY=3.5, extentZ=18, shapeName=dyn_SkirtH04_box, center=<0 1.2761 -8.4305>, rotation=<rot 0 0 -5.5799>}]
 dyn_SkirtH05 [{shapeType=Box, extentX=22.5, extentY=4.5, extentZ=18, shapeName=dyn_SkirtH05_box, center=<0 3.6458 -8.1659>, rotation=<rot 0 0 -12.6459>}]
 dyn_SkirtH06 [{
   shapeType=Capsule
   length=19.7603
   radius=9.75
   shapeName=dyn_SkirtH06_capsule
   center=<0 1.5221 -3.9088>
   rotation=<rot 78.5145 -72.4721 -72.8021>
  }]
 dyn_SkirtJ01 [{
   shapeType=Capsule
   length=10.1066
   radius=6.0973
   shapeName=dyn_SkirtJ01_capsule
   center=<0 0 -4.9582>
   rotation=<rot -89.5211 -179.9979 179.9979>
  }]
 dyn_SkirtJ02 [{shapeType=Box, extentX=23.3552, extentY=5.9738, extentZ=14.4279, shapeName=dyn_SkirtB02_box, center=<0 -0.0328 -4.1981>}]
 dyn_SkirtJ03 [{shapeType=Box, extentX=23.3552, extentY=5.9738, extentZ=14.4279, shapeName=dyn_SkirtB03_box, center=<0 -0.0328 -4.1981>}]
 dyn_SkirtJ04 [{shapeType=Box, extentX=24.0761, extentY=7.3649, extentZ=19, shapeName=dyn_SkirtJ04_box, center=<0 0 -5.5058>}]
 dyn_SkirtJ05 [{shapeType=Box, extentX=24.0761, extentY=7.3649, extentZ=19, shapeName=dyn_SkirtJ05_box, center=<0 0 -5.5058>}]
 dyn_SkirtJ06 [{shapeType=Box, extentX=22.6914, extentY=7.2465, extentZ=23, shapeName=dyn_SkirtJ06_box, center=<0 0 -8.9173>}]
 foot_l [{shapeType=Capsule, length=14, radius=8.5, shapeName=foot_r_capsule, center=<-3.0909 8.2397 1.0623>, rotation=<rot 0 0 82.6562>}]
 foot_r [{shapeType=Capsule, length=14, radius=8.5, shapeName=foot_r_capsule, center=<3.5892 -7.5638 -0.9746>, rotation=<rot 0 0 82.6562>}]
 pelvis [{shapeType=Capsule, length=16.0527, radius=8.0414, shapeName=pelvis_capsule, center=<-2.3033 -0.0002 0.2457>, rotation=<rot 1.9231 0 0>}]
 root [
  {shapeType=Capsule, length=0, radius=87.0435, shapeName=root_capsule, center=<0.1551 5.6254 -65.1086>, rotation=<rot 0.2137 -0.0005 0>}
  {shapeType=Box, extentX=179.5, extentY=116, extentZ=7.5, shapeName=root_box, center=<-0.0001 -16.4994 -0.1277>}
 ]
 spine_01 [{shapeType=Capsule, length=5.5221, radius=8.9779, shapeName=spine_01_capsule, center=<1.1973 0 0>}]
 thigh_twist_01_l [{
   shapeType=Capsule
   length=32.75
   radius=7.5
   shapeName=thigh_r_capsule
   center=<3.8947 -0.0001 -0.0564>
   rotation=<rot -89.1701 -179.9991 179.9991>
  }]
 thigh_twist_01_r [{
   shapeType=Capsule
   length=32.75
   radius=7.5
   shapeName=thigh_r_capsule
   center=<-3.4223 0 0.0496>
   rotation=<rot -89.1701 -179.9991 179.9991>
  }]
}
constraints [
 {bone1Name=dyn_SkirtJ01, bone2Name=spine_01, swing1LimitDegrees=30, swing2LimitDegrees=30, twistLimitDegrees=60}
 {bone1Name=dyn_SkirtJ02, bone2Name=dyn_SkirtJ01, swing1LimitDegrees=30, swing2LimitDegrees=30, twistLimitDegrees=60}
 {bone1Name=dyn_SkirtJ03, bone2Name=dyn_SkirtJ02, swing1LimitDegrees=30, swing2LimitDegrees=30, twistLimitDegrees=60}
 {bone1Name=dyn_SkirtJ04, bone2Name=dyn_SkirtJ03, swing1LimitDegrees=30, swing2LimitDegrees=30, twistLimitDegrees=60}
 {bone1Name=dyn_SkirtJ05, bone2Name=dyn_SkirtJ04, swing1LimitDegrees=30, swing2LimitDegrees=30, twistLimitDegrees=60}
 {bone1Name=dyn_SkirtJ06, bone2Name=dyn_SkirtJ05, swing1LimitDegrees=30, swing2LimitDegrees=30, twistLimitDegrees=60}
 {bone1Name=dyn_SkirtH01, bone2Name=spine_01, swing1LimitDegrees=20, swing2LimitDegrees=20, twistLimitDegrees=60}
 {bone1Name=dyn_SkirtH02, bone2Name=dyn_SkirtH01, swing1LimitDegrees=20, swing2LimitDegrees=20, twistLimitDegrees=60}
 {bone1Name=dyn_SkirtH03, bone2Name=dyn_SkirtH02, swing1LimitDegrees=20, swing2LimitDegrees=20, twistLimitDegrees=60}
 {bone1Name=dyn_SkirtH04, bone2Name=dyn_SkirtH03, swing1LimitDegrees=20, swing2LimitDegrees=20, twistLimitDegrees=60}
 {bone1Name=dyn_SkirtH05, bone2Name=dyn_SkirtH04, swing1LimitDegrees=20, swing2LimitDegrees=20, twistLimitDegrees=60}
 {bone1Name=dyn_SkirtH06, bone2Name=dyn_SkirtH05, swing1LimitDegrees=20, swing2LimitDegrees=20, twistLimitDegrees=60}
 {bone1Name=dyn_SkirtG01, bone2Name=spine_01, swing1LimitDegrees=20, swing2LimitDegrees=20, twistLimitDegrees=60}
 {bone1Name=dyn_SkirtG02, bone2Name=dyn_SkirtG01, swing1LimitDegrees=10, swing2LimitDegrees=10, twistLimitDegrees=30}
 {bone1Name=dyn_SkirtG03, bone2Name=dyn_SkirtG02, swing1LimitDegrees=10, swing2LimitDegrees=10, twistLimitDegrees=30}
 {bone1Name=dyn_SkirtG04, bone2Name=dyn_SkirtG03, swing1LimitDegrees=20, swing2LimitDegrees=20, twistLimitDegrees=60}
 {bone1Name=dyn_SkirtG05, bone2Name=dyn_SkirtG04, swing1LimitDegrees=20, swing2LimitDegrees=20, twistLimitDegrees=60}
 {bone1Name=dyn_SkirtG06, bone2Name=dyn_SkirtG05, swing1LimitDegrees=20, swing2LimitDegrees=20, twistLimitDegrees=60}
 {
  bone1Name=dyn_SkirtF01
  bone2Name=spine_01
  swing1LimitDegrees=60
  swing1Motion=Locked
  swing2LimitDegrees=60
  swing2Motion=Locked
  twistLimitDegrees=60
  twistMotion=Locked
 }
 {
  bone1Name=dyn_SkirtF02
  bone2Name=dyn_SkirtF01
  swing1LimitDegrees=60
  swing1Motion=Locked
  swing2LimitDegrees=60
  swing2Motion=Locked
  twistLimitDegrees=60
  twistMotion=Locked
 }
 {
  bone1Name=dyn_SkirtF03
  bone2Name=dyn_SkirtF02
  swing1LimitDegrees=60
  swing1Motion=Locked
  swing2LimitDegrees=60
  swing2Motion=Locked
  twistLimitDegrees=60
  twistMotion=Locked
 }
 {bone1Name=dyn_SkirtF04, bone2Name=dyn_SkirtF03, swing1LimitDegrees=10, swing2LimitDegrees=10, twistLimitDegrees=30}
 {bone1Name=dyn_SkirtF05, bone2Name=dyn_SkirtF04, swing1LimitDegrees=10, swing2LimitDegrees=10, twistLimitDegrees=30}
 {bone1Name=dyn_SkirtF06, bone2Name=dyn_SkirtF05, swing1LimitDegrees=10, swing2LimitDegrees=10, twistLimitDegrees=30}
 {bone1Name=dyn_SkirtE01, bone2Name=spine_01, swing1LimitDegrees=20, swing2LimitDegrees=20, twistLimitDegrees=60}
 {bone1Name=dyn_SkirtE02, bone2Name=dyn_SkirtE01, swing1LimitDegrees=10, swing2LimitDegrees=10, twistLimitDegrees=30}
 {bone1Name=dyn_SkirtE03, bone2Name=dyn_SkirtE02, swing1LimitDegrees=10, swing2LimitDegrees=10, twistLimitDegrees=30}
 {bone1Name=dyn_SkirtE04, bone2Name=dyn_SkirtE03, swing1LimitDegrees=10, swing2LimitDegrees=10, twistLimitDegrees=30}
 {bone1Name=dyn_SkirtE05, bone2Name=dyn_SkirtE04, swing1LimitDegrees=10, swing2LimitDegrees=10, twistLimitDegrees=30}
 {bone1Name=dyn_SkirtE06, bone2Name=dyn_SkirtE05, swing1LimitDegrees=10, swing2LimitDegrees=10, twistLimitDegrees=30}
 {bone1Name=dyn_SkirtD01, bone2Name=spine_01, swing1LimitDegrees=20, swing2LimitDegrees=20, twistLimitDegrees=60}
 {bone1Name=dyn_SkirtD02, bone2Name=dyn_SkirtD01, swing1LimitDegrees=20, swing2LimitDegrees=20, twistLimitDegrees=60}
 {bone1Name=dyn_SkirtD03, bone2Name=dyn_SkirtD02, swing1LimitDegrees=20, swing2LimitDegrees=20, twistLimitDegrees=60}
 {bone1Name=dyn_SkirtD04, bone2Name=dyn_SkirtD03, swing1LimitDegrees=20, swing2LimitDegrees=20, twistLimitDegrees=60}
 {bone1Name=dyn_SkirtD05, bone2Name=dyn_SkirtD04, swing1LimitDegrees=20, swing2LimitDegrees=20, twistLimitDegrees=60}
 {bone1Name=dyn_SkirtD06, bone2Name=dyn_SkirtD05, swing1LimitDegrees=20, swing2LimitDegrees=20, twistLimitDegrees=60}
 {bone1Name=dyn_SkirtB01, bone2Name=spine_01, swing1LimitDegrees=30, swing2LimitDegrees=30, twistLimitDegrees=60}
 {bone1Name=dyn_SkirtB02, bone2Name=dyn_SkirtB01, swing1LimitDegrees=30, swing2LimitDegrees=30, twistLimitDegrees=60}
 {bone1Name=dyn_SkirtB03, bone2Name=dyn_SkirtB02, swing1LimitDegrees=30, swing2LimitDegrees=30, twistLimitDegrees=60}
 {bone1Name=dyn_SkirtB04, bone2Name=dyn_SkirtB03, swing1LimitDegrees=30, swing2LimitDegrees=30, twistLimitDegrees=60}
 {bone1Name=dyn_SkirtB05, bone2Name=dyn_SkirtB04, swing1LimitDegrees=30, swing2LimitDegrees=30, twistLimitDegrees=60}
 {bone1Name=dyn_SkirtB06, bone2Name=dyn_SkirtB05, swing1LimitDegrees=30, swing2LimitDegrees=30, twistLimitDegrees=60}
 {bone1Name=dyn_SkirtE05, bone2Name=dyn_SkirtF05, swing1LimitDegrees=45, swing2LimitDegrees=45, swing2Motion=Locked, twistLimitDegrees=45}
 {bone1Name=dyn_SkirtG05, bone2Name=dyn_SkirtF05, swing1LimitDegrees=45, swing2LimitDegrees=45, swing2Motion=Locked, twistLimitDegrees=45}
 {bone1Name=dyn_SkirtE04, bone2Name=dyn_SkirtF04, swing1LimitDegrees=45, swing2LimitDegrees=45, swing2Motion=Locked, twistLimitDegrees=45}
 {bone1Name=dyn_SkirtD06, bone2Name=dyn_SkirtE06, swing1LimitDegrees=45, swing2LimitDegrees=10, swing2Motion=Locked, twistLimitDegrees=45}
 {bone1Name=dyn_SkirtD05, bone2Name=dyn_SkirtE05, swing1LimitDegrees=45, swing2LimitDegrees=10, swing2Motion=Locked, twistLimitDegrees=45}
 {bone1Name=dyn_SkirtD04, bone2Name=dyn_SkirtE04, swing1LimitDegrees=45, swing2LimitDegrees=10, swing2Motion=Locked, twistLimitDegrees=45}
 {
  bone1Name=dyn_SkirtB06
  bone2Name=dyn_SkirtD06
  swing1LimitDegrees=30
  swing2LimitDegrees=10
  swing2Motion=Free
  twistLimitDegrees=30
  twistMotion=Free
 }
 {bone1Name=dyn_SkirtB05, bone2Name=dyn_SkirtD05, swing1LimitDegrees=30, swing2LimitDegrees=10, twistLimitDegrees=30}
 {bone1Name=dyn_SkirtB04, bone2Name=dyn_SkirtD04, swing1LimitDegrees=30, swing2LimitDegrees=10, twistLimitDegrees=30}
 {bone1Name=dyn_SkirtH06, bone2Name=dyn_SkirtG06, swing1LimitDegrees=45, swing2LimitDegrees=10, swing2Motion=Locked, twistLimitDegrees=45}
 {bone1Name=dyn_SkirtH05, bone2Name=dyn_SkirtG05, swing1LimitDegrees=45, swing2LimitDegrees=10, swing2Motion=Locked, twistLimitDegrees=45}
 {bone1Name=dyn_SkirtH04, bone2Name=dyn_SkirtG04, swing1LimitDegrees=45, swing2LimitDegrees=10, swing2Motion=Locked, twistLimitDegrees=45}
 {
  bone1Name=dyn_SkirtJ06
  bone2Name=dyn_SkirtH06
  swing1LimitDegrees=30
  swing2LimitDegrees=10
  swing2Motion=Free
  twistLimitDegrees=30
  twistMotion=Free
 }
 {bone1Name=dyn_SkirtJ05, bone2Name=dyn_SkirtH05, swing1LimitDegrees=30, swing2LimitDegrees=10, twistLimitDegrees=30}
 {bone1Name=dyn_SkirtJ04, bone2Name=dyn_SkirtH04, swing1LimitDegrees=30, swing2LimitDegrees=10, twistLimitDegrees=30}
 {
  bone1Name=dyn_SkirtE03
  bone2Name=dyn_SkirtF03
  swing1LimitDegrees=45
  swing1Motion=Free
  swing2LimitDegrees=45
  swing2Motion=Free
  twistLimitDegrees=45
  twistMotion=Free
 }
 {bone1Name=dyn_SkirtG03, bone2Name=dyn_SkirtF03, swing1LimitDegrees=45, swing2LimitDegrees=45, swing2Motion=Locked, twistLimitDegrees=45}
 {
  bone1Name=dyn_SkirtE02
  bone2Name=dyn_SkirtF02
  swing1LimitDegrees=45
  swing1Motion=Free
  swing2LimitDegrees=45
  swing2Motion=Free
  twistLimitDegrees=45
  twistMotion=Free
 }
 {bone1Name=dyn_SkirtG02, bone2Name=dyn_SkirtF02, swing1LimitDegrees=45, swing2LimitDegrees=45, swing2Motion=Locked, twistLimitDegrees=45}
 {
  bone1Name=dyn_SkirtE01
  bone2Name=dyn_SkirtF01
  swing1LimitDegrees=45
  swing1Motion=Free
  swing2LimitDegrees=45
  swing2Motion=Free
  twistLimitDegrees=45
  twistMotion=Free
 }
 {bone1Name=dyn_SkirtG01, bone2Name=dyn_SkirtF01, swing1LimitDegrees=45, swing2LimitDegrees=45, swing2Motion=Locked, twistLimitDegrees=45}
 {bone1Name=dyn_SkirtD03, bone2Name=dyn_SkirtE03, swing1LimitDegrees=45, swing2LimitDegrees=10, swing2Motion=Locked, twistLimitDegrees=45}
 {bone1Name=dyn_SkirtD02, bone2Name=dyn_SkirtE02, swing1LimitDegrees=45, swing2LimitDegrees=45, swing2Motion=Locked, twistLimitDegrees=45}
 {
  bone1Name=dyn_SkirtD01
  bone2Name=dyn_SkirtE01
  swing1LimitDegrees=45
  swing1Motion=Free
  swing2LimitDegrees=45
  swing2Motion=Locked
  twistLimitDegrees=45
  twistMotion=Free
 }
 {bone1Name=dyn_SkirtB03, bone2Name=dyn_SkirtD03, swing1LimitDegrees=45, swing2LimitDegrees=45, swing2Motion=Locked, twistLimitDegrees=45}
 {bone1Name=dyn_SkirtB02, bone2Name=dyn_SkirtD02, swing1LimitDegrees=45, swing2LimitDegrees=45, swing2Motion=Locked, twistLimitDegrees=45}
 {bone1Name=dyn_SkirtH03, bone2Name=dyn_SkirtG03, swing1LimitDegrees=45, swing2LimitDegrees=10, swing2Motion=Locked, twistLimitDegrees=45}
 {bone1Name=dyn_SkirtH02, bone2Name=dyn_SkirtG02, swing1LimitDegrees=45, swing2LimitDegrees=10, swing2Motion=Locked, twistLimitDegrees=45}
 {
  bone1Name=dyn_SkirtH01
  bone2Name=dyn_SkirtG01
  swing1LimitDegrees=45
  swing1Motion=Locked
  swing2LimitDegrees=45
  swing2Motion=Locked
  twistLimitDegrees=45
 }
 {bone1Name=dyn_SkirtJ03, bone2Name=dyn_SkirtH03, swing1LimitDegrees=45, swing2LimitDegrees=45, swing2Motion=Locked, twistLimitDegrees=45}
 {bone1Name=dyn_SkirtJ02, bone2Name=dyn_SkirtH02, swing1LimitDegrees=45, swing2LimitDegrees=45, swing2Motion=Locked, twistLimitDegrees=45}
 {bone1Name=dyn_SkirtG04, bone2Name=dyn_SkirtF04, swing1LimitDegrees=45, swing2LimitDegrees=45, swing2Motion=Locked, twistLimitDegrees=45}
 {bone1Name=dyn_SkirtF06, bone2Name=dyn_SkirtG06, swing1LimitDegrees=45, swing2LimitDegrees=45, swing2Motion=Locked, twistLimitDegrees=45}
 {bone1Name=dyn_SkirtF06, bone2Name=dyn_SkirtE06, swing1LimitDegrees=45, swing2LimitDegrees=45, swing2Motion=Locked, twistLimitDegrees=45}
]
